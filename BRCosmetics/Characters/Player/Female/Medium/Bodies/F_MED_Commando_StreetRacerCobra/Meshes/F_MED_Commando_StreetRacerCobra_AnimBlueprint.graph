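# Animation Blueprint for the female Commando StreetRacerCobra body cosmetic that adds secondary physics motion to its zippers and tongue bones.
# It copies the attached parent mesh's pose, then chains seven single-bone Anim Dynamics nodes with constrained angular limits.
# Five zipper nodes simulate in actor space with an LOD threshold, the right arm's at low alpha.
# Two tongue nodes use reduced gravity and overridden angular bias; the event graph has no logic.
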AnimBlueprint /BRCosmetics/Characters/Player/Female/Medium/Bodies/F_MED_Commando_StreetRacerCobra/Meshes/F_MED_Commando_StreetRacerCobra_AnimBlueprint
target_skeleton=&/F_MED_Commando_StreetRacerCobra_Skeleton
graphs [
 AnimGraph {nodes [
   n1 = Anim Dynamics - Bone: dyn_foot_zipper_R(ComponentPose=Local To Component(LocalPose=Copy Pose From Mesh() {bUseAttachedParent=true}), Alpha=1, ExternalForce="0,0,0") {
    bGravityOverrideInSimSpace=true
    lODThreshold=1
    numSolverIterationsPostUpdate=4
    numSolverIterationsPreUpdate=16
    simulationSpace=Actor
    alphaBoolBlend {blendOption=Linear}
    boundBone {boneName=dyn_foot_zipper_R}
    physicsBodyDefinitions [{
      sphereCollisionRadius=0
      boundBone {boneName=dyn_foot_zipper_R}
      boxExtents=<5 5 5>
      constraintSetup {angularTargetAxis=AxisZ, twistAxis=AxisZ, angularLimitsMax=<5 0 0>, angularLimitsMin=<-90 -30 0>, angularTarget=<0 0 0>}
      localJointOffset=<0 0 -0.5>
     }]
   }
   n2 = Anim Dynamics - Bone: dyn_foot_zipper_L(ComponentPose=@n1, Alpha=1, ExternalForce="0,0,0") {
    bGravityOverrideInSimSpace=true
    lODThreshold=1
    numSolverIterationsPostUpdate=4
    numSolverIterationsPreUpdate=16
    simulationSpace=Actor
    alphaBoolBlend {blendOption=Linear}
    boundBone {boneName=dyn_foot_zipper_L}
    physicsBodyDefinitions [{
      sphereCollisionRadius=0
      boundBone {boneName=dyn_foot_zipper_L}
      boxExtents=<5 5 5>
      constraintSetup {angularTargetAxis=AxisZ, twistAxis=AxisZ, angularLimitsMax=<5 0 0>, angularLimitsMin=<-90 -30 0>, angularTarget=<0 0 0>}
      localJointOffset=<0 0 -0.5>
     }]
   }
   n3 = Anim Dynamics - Bone: dyn_zipper(ComponentPose=@n2, Alpha=1) {
    bGravityOverrideInSimSpace=true
    lODThreshold=1
    numSolverIterationsPostUpdate=4
    numSolverIterationsPreUpdate=16
    simulationSpace=Actor
    alphaBoolBlend {blendOption=Linear}
    boundBone {boneName=dyn_zipper}
    physicsBodyDefinitions [{
      sphereCollisionRadius=0
      boundBone {boneName=dyn_zipper}
      boxExtents=<5 5 5>
      constraintSetup {angularTargetAxis=AxisZ, twistAxis=AxisZ, angularLimitsMax=<5 30 0>, angularLimitsMin=<-90 -30 0>, angularTarget=<0 0 0>}
      localJointOffset=<0 0 -0.5>
     }]
   }
   n4 = Anim Dynamics - Bone: dyn_arm_zipper_R(ComponentPose=@n3, Alpha=0.15, ExternalForce="0,0,0") {
    bGravityOverrideInSimSpace=true
    lODThreshold=1
    numSolverIterationsPostUpdate=4
    numSolverIterationsPreUpdate=16
    simulationSpace=Actor
    alphaBoolBlend {blendOption=Linear}
    boundBone {boneName=dyn_arm_zipper_R}
    physicsBodyDefinitions [{
      sphereCollisionRadius=0
      boundBone {boneName=dyn_arm_zipper_R}
      boxExtents=<5 5 5>
      constraintSetup {angularTargetAxis=AxisZ, twistAxis=AxisZ, angularLimitsMax=<5 30 30>, angularLimitsMin=<-90 0 -30>, angularTarget=<0 0 0>}
      localJointOffset=<0 0 -0.5>
     }]
   }
   n5 = Anim Dynamics - Bone: dyn_arm_zipper_L(ComponentPose=@n4, Alpha=1, ExternalForce="0,0,0") {
    bGravityOverrideInSimSpace=true
    lODThreshold=1
    numSolverIterationsPostUpdate=4
    numSolverIterationsPreUpdate=16
    simulationSpace=Actor
    alphaBoolBlend {blendOption=Linear}
    boundBone {boneName=dyn_arm_zipper_L}
    physicsBodyDefinitions [{
      sphereCollisionRadius=0
      boundBone {boneName=dyn_arm_zipper_L}
      boxExtents=<5 5 5>
      constraintSetup {angularTargetAxis=AxisZ, twistAxis=AxisZ, angularLimitsMax=<5 0 30>, angularLimitsMin=<-90 -30 -30>, angularTarget=<0 0 0>}
      localJointOffset=<0 0 -0.5>
     }]
   }
   n6 = Anim Dynamics - Bone: dyn_tongue_l(ComponentPose=@n5, Alpha=1, ExternalForce="0,0,0") {
    angularBiasOverride=1
    bGravityOverrideInSimSpace=true
    bOverrideAngularBias=true
    gravityScale=0.1
    alphaBoolBlend {blendOption=Linear}
    boundBone {boneName=dyn_tongue_l}
    physicsBodyDefinitions [{
      sphereCollisionRadius=0
      boundBone {boneName=dyn_tongue_l}
      boxExtents=<5 5 5>
      constraintSetup {
       angularTargetAxis=AxisY
       twistAxis=AxisY
       angularLimitsMax=<0 30 0>
       angularLimitsMin=<0 -30 0>
       angularTarget=<0 0 0>
       linearAxesMax=<0 0 3>
      }
      localJointOffset=<0 0 -0.1>
     }]
   }
   n7 = Anim Dynamics - Bone: dyn_tongue_r(ComponentPose=@n6, Alpha=1, ExternalForce="0,0,0") {
    angularBiasOverride=1
    bGravityOverrideInSimSpace=true
    bOverrideAngularBias=true
    gravityScale=0.1
    alphaBoolBlend {blendOption=Linear}
    boundBone {boneName=dyn_tongue_r}
    physicsBodyDefinitions [{
      sphereCollisionRadius=0
      boundBone {boneName=dyn_tongue_r}
      boxExtents=<5 5 5>
      constraintSetup {
       angularTargetAxis=AxisY
       twistAxis=AxisY
       angularLimitsMax=<0 30 0>
       angularLimitsMin=<0 -30 0>
       angularTarget=<0 0 0>
       linearAxesMax=<0 0 3>
      }
      localJointOffset=<0 0 -0.1>
     }]
   }
   Output Pose(Result=Component To Local(ComponentPose=@n7)) {layerGroup=None}
  ]}
 EventGraph {nodes [Event BlueprintUpdateAnimation()]}
]
variables []
